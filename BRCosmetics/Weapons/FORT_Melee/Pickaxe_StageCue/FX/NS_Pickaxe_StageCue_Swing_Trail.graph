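# Niagara swing trail for the StageCue pickaxe, drawn as two ribbons, Trail_Head and Trail_End.
# Each emitter uses FN AnimTrail modules to follow its own bone or socket pair on the skeletal mesh.
# User.SwingSectionIndex selects each ribbon's lifetime, zeroing it to hide that ribbon in certain swing sections.
# Time of day drives color and alpha, with alpha also fading over particle life and by System.TrailFade.
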NiagaraSystem /BRCosmetics/Weapons/FORT_Melee/Pickaxe_StageCue/FX/NS_Pickaxe_StageCue_Swing_Trail
dependencies {
 usedDataInterfaces [&NiagaraDataInterfaceCurve, &NiagaraDataInterfaceSkeletalMesh]
 usedDynamicInputs [&self:SectionIndexFloatSelector]
 usedModules [
  &self:SystemSpawnScript.SystemScriptSource.SystemScriptGraph.NiagaraNodeAssignment_0.SetVariables_#1
  &SystemState
  &self:SystemSpawnScript.SystemScriptSource.SystemScriptGraph.NiagaraNodeAssignment_1.SetVariables_#2
 ]
}
emitters {
 Trail_End {
  emitterUpdateScript [
   &EmitterState()
   &FN_SpawnAnimTrail()
  ]
  particleSpawnScript [
   &FN_InitializeAnimTrailBoneIndecies(First Bone or Socket Index=2, Second Bone or Socket Index=3)
   &FN_InitializeAnimTrail(
    Color=&FNTimeOfDayColor(Night Color=<rgba 0.0863 0.2 0.6 1>, Morning Color=<rgba 1 1 1 1>, Day Color=<rgba 0.7081 0.8306 1 1>, Evening Color=<rgba 1 1 1 1>)
    Lifetime=&MakeCustomFloatFromBool(
     Bool=&Boolean_OrOperation(
      Bool1=&SetBoolByIntComparison(A=$User.SwingSectionIndex, B=2, Comparison Type=::NewEnumerator4)
      Bool2=&SetBoolByIntComparison(A=$User.SwingSectionIndex, B=3, Comparison Type=::NewEnumerator4)
     )
     True Float=0.2
     False Float=0
    )
    SkeletalMesh=$System.Skeletal Mesh
   )
  ]
  particleUpdateScript [
   &ParticleState()
   &FN_UpdateAnimTrail(SkeletalMesh=$System.Skeletal Mesh)
   &ScaleColor(
    Scale Alpha=&FloatFromCurve(
     FloatCurve=DataInterface() {propertyValues {Curve {keys [
         {arriveTangent=-2.465, interpMode=RCIM_Cubic, leaveTangent=-2.465, tangentMode=RCTM_User, time=0, value=1}
         {interpMode=RCIM_Cubic, tangentMode=RCTM_User, time=1, value=0}
        ]}}}
     Scale Curve=&Multiply_Float(
      A=&FNTimeOfDayScalar(Night Scale=0.5, Morning Scale=1, Day Scale=0.4, Evening Scale=1)
      B=$System.TrailFade
     )
    )
    ScaleA=true
    ScaleRGB=true
   )
  ]
  rendererClasses [&NiagaraRibbonRendererProperties]
  renderers [{rendererClass=&NiagaraRibbonRendererProperties}]
 }
 Trail_Head {
  emitterUpdateScript [
   &EmitterState()
   &FN_SpawnAnimTrail()
  ]
  particleSpawnScript [
   &FN_InitializeAnimTrailBoneIndecies()
   &FN_InitializeAnimTrail(
    Color=&FNTimeOfDayColor(Night Color=<rgba 0.0863 0.2 0.6 1>, Morning Color=<rgba 1 1 1 1>, Day Color=<rgba 0.7081 0.8306 1 1>, Evening Color=<rgba 1 1 1 1>)
    Lifetime=&MakeCustomFloatFromBool(
     Bool=&SetBoolByIntComparison(A=$User.SwingSectionIndex, B=2, Comparison Type=::NewEnumerator4)
     True Float=0
     False Float=0.2
    )
    SkeletalMesh=$System.Skeletal Mesh
   )
  ]
  particleUpdateScript [
   &ParticleState()
   &FN_UpdateAnimTrail(SkeletalMesh=$System.Skeletal Mesh)
   &ScaleColor(
    Scale Alpha=&FloatFromCurve(
     FloatCurve=DataInterface() {propertyValues {Curve {keys [
         {arriveTangent=-2.465, interpMode=RCIM_Cubic, leaveTangent=-2.465, tangentMode=RCTM_User, time=0, value=1}
         {interpMode=RCIM_Cubic, tangentMode=RCTM_User, time=1, value=0}
        ]}}}
     Scale Curve=&Multiply_Float(
      A=&FNTimeOfDayScalar(Night Scale=0.5, Morning Scale=1, Day Scale=0.4, Evening Scale=1)
      B=$System.TrailFade
     )
    )
    Scale RGB=&VectorFromFloat(Value=0.5)
    ScaleA=true
    ScaleRGB=true
   )
  ]
  rendererClasses [&NiagaraRibbonRendererProperties]
  renderers [{rendererClass=&NiagaraRibbonRendererProperties}]
 }
}
summary {
 systemName=NS_Pickaxe_StageCue_Swing_Trail
 emitters [
  {emitterName=Trail_Head, rendererClasses [&NiagaraRibbonRendererProperties]}
  {emitterName=Trail_End, rendererClasses [&NiagaraRibbonRendererProperties]}
 ]
 userVariables [
  User.RibbonIDOffset {type=NiagaraInt32}
  User.SwingSectionIndex {type=NiagaraInt32}
 ]
}
