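# Physics asset for the StageCue pickaxe mesh, a cosmetic two-handed melee harvesting weapon.
# It defines a single capsule body on the Melee_TwoHanded_Handle bone, sized and offset to run along the length of the axe.
# There are no constraints, so the weapon simulates and collides as one rigid piece.
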PhysicsAsset /BRCosmetics/Weapons/FORT_Melee/Pickaxe_StageCue/Meshes/StageCue_Axe_PhysicsAsset
bodies {Melee_TwoHanded_Handle [{shapeType=Capsule, length=86.2454, radius=2.6803, center=<-0.008 -1.0058 45.7902>, rotation=<rot 0.0002 -90 0>}]}
constraints []
