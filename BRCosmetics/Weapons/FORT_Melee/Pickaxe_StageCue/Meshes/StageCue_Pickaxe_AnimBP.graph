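# Animation Blueprint for the StageCue pickaxe skeleton, a melee weapon cosmetic.
# The AnimGraph passes an input pose through the DefaultSlot montage slot to the output pose, so slot animations can play on the pickaxe mesh.
# It holds no custom logic or variables: the update event and TryGetPawnOwner node are unconnected, and the OnUpdate function is an empty stub.
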AnimBlueprint /BRCosmetics/Weapons/FORT_Melee/Pickaxe_StageCue/Meshes/StageCue_Pickaxe_AnimBP
target_skeleton=&/StageCue_Axe_Skeleton
graphs [
 AnimGraph {nodes [Output Pose(Result=Slot 'DefaultSlot'(Source=Input Pose())) {layerGroup=None}]}
 EventGraph {nodes [
   TryGetPawnOwner()
   Event BlueprintUpdateAnimation()
  ]}
 OnUpdate {nodes [
   OnUpdate!() -> (then, Context:"Anim Update Context Structure (by ref)", Node:"Anim Node Reference Structure (by ref)") {is=title K2Node_FunctionEntry}
   > Return Node()
  ]}
]
variables []
